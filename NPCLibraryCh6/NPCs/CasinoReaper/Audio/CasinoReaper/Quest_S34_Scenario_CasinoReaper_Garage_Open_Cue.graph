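# Sound cue for the Casino Reaper NPC's quest garage-open audio in the S34 scenario.
# A single wave player feeds a mixer to play the garage-open sound wave at boosted volume in the quests sound class.
# It borrows a supply-drop ambient loop attenuation asset for spatial falloff.
# Concurrency override allows only one instance per owner and blocks new plays while one is active.
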SoundCue /NPCLibraryCh6/NPCs/CasinoReaper/Audio/CasinoReaper/Quest_S34_Scenario_CasinoReaper_Garage_Open_Cue
root=Mixer_0
nodes [
 {id=Mixer_0, children [WavePlayer_1], properties {inputVolume [1]}}
 {
  id=WavePlayer_1
  properties {soundWaveAssetPtr=&/Quest_S34_Scenario_CasinoReaper_Garage_Open}
 }
]
properties {
 attenuationSettings=/Game/Athena/Sounds/SupplyDrops/Fort_HiddenStar_Ambient_Loop_Att
 bOverrideConcurrency=true
 maxDistance=2500
 soundClassObject=&Fort_Quests
 volumeMultiplier=3
 concurrencyOverrides {bLimitToOwner=true, maxCount=1, resolutionRule=PreventNew}
}
