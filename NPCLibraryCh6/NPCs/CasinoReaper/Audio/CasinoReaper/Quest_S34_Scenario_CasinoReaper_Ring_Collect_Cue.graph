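# Sound cue for collecting a ring in the Quest S34 CasinoReaper scenario, played under the non-critical in-game UI sound class.
# A Random node picks one of three equally weighted collect waves and feeds it through a Modulator into a Mixer whose second input is empty.
# A concurrency override allows one instance per owner and blocks new plays while it is active.
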SoundCue /NPCLibraryCh6/NPCs/CasinoReaper/Audio/CasinoReaper/Quest_S34_Scenario_CasinoReaper_Ring_Collect_Cue
root=Mixer_0
nodes [
 {id=Mixer_0, children [Modulator_0,null], properties {inputVolume [1,1]}}
 {id=Modulator_0, children [Random_0]}
 {id=Random_0, children [WavePlayer_0,WavePlayer_2,WavePlayer_4], properties {weights [1,1,1]}}
 {
  id=WavePlayer_0
  properties {soundWaveAssetPtr=&/Quest_S34_Scenario_CasinoReaper_Ring_Collect_01}
 }
 {
  id=WavePlayer_2
  properties {soundWaveAssetPtr=&/Quest_S34_Scenario_CasinoReaper_Ring_Collect_02}
 }
 {
  id=WavePlayer_4
  properties {soundWaveAssetPtr=&/Quest_S34_Scenario_CasinoReaper_Ring_Collect_03}
 }
]
properties {
 bOverrideConcurrency=true
 maxDistance=2097152
 soundClassObject=&Fort_UI_InGame_NonCritical
 volumeMultiplier=1.5
 concurrencyOverrides {bLimitToOwner=true, maxCount=1, resolutionRule=PreventNew}
}
